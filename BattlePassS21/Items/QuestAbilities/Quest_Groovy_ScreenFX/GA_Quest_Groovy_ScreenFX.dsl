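; Quest gameplay ability that plays a psychedelic screen effect on the owning player, which comments tie to interacting with an Object of Power.
; On event-triggered activation, it calls the parent, then starts the Season 21 psychedelic camera shake in camera-local space through the pawn's player controller.
; It ends immediately after starting the shake, and the end-ability handler is empty.
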(event Ability|EventActivateAbilityFromEvent
  (params (EventData :type Gameplay%20Event%20Data%20Structure%20%28by%20ref%29))
  ;[1 1 1] Activate ability - listen for when player interacts with an Object of Power
  (|CallParentFunction EventData)
  ;;[1 1 1] Play Camera Anim of Psychedelic Effect
  (Game|Feedback|ClientStartCameraShake :self (Fortnite|GetFortPlayerControllerfromActor (Variables|GAVars|Data|GetOwningPlayerPawn)) :Shake "/Game/Effects/Fort_Effects/Camera_Effects/BP_CameraShake_S21_PsychedelicEffect.BP_CameraShake_S21_PsychedelicEffect_C" :Scale 1.0 :PlaySpace "CameraLocal")
  (Ability|EndAbility))

(event Ability|EventOnEndAbility
  (params (bWasCancelled :type Boolean)))
